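# Masked, two-sided material for the Llama backpack cosmetic, rendering an overlay texture as scaled emissive over zero base color and specular.
# The green channel of mask texture M drives dithered opacity.
# The attributes pass through the shared character brightness/scale balance function, then the character PostFX function, fed pre-skinned position via a vertex interpolator for pixel shading.
# The opacity mask is set again after PostFX.
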Material /BRCosmetics/Accessories/FORT_Backpacks/Football_Backpacks/Materials/M_Llama_Backpack
graph {
 properties {
  bAutomaticallySetUsageInEditor=false
  bUseLightmapDirectionality=false
  bUseMaterialAttributes=true
  bUsedWithMeshDeformer=true
  bUsedWithSkeletalMesh=true
  bUsedWithStaticLighting=true
  blendMode=BLEND_Masked
  ditherOpacityMask=true
  pixelDepthOffsetMode=PDOM_Legacy
  twoSided=true
  usedShadingModels=Default Lit
 }
 nodes [
  n1 = Constant()
  n2 = TextureSampleParameter2D() {group=Texture, parameterName=M, samplerType=SAMPLERTYPE_Masks, texture=&T_BlackMask}
  n3 = SetMaterialAttributes!(
   MaterialAttributes
   Base Color=@n1
   Emissive Color=Multiply(
    A=TextureSampleParameter2D() {group=Texture, parameterName=Diffuse Texture Overlay, texture=&T_WhitesRGB}.RGB
    B=ScalarParameter() {defaultValue=2, group=Parameters, parameterName=EmissiveMult}
   ) {constB=2}
   Opacity Mask=@n2.G
   Shading Model
   Specular=@n1
  ) {attributeSetTypes [
    #1
    #2
    #3
    #4
    #5
   ]}
  n4 = MaterialFunctionCall!(Material=@n3) {materialFunction=/Game/Packages/Fortress_SharedMaterials/Base_Material_Functions/CharacterBrightnessandScale_Balance}
  n5 = PreSkinnedPosition()
  n6 = MaterialFunctionCall(
   AO=Constant() {r=1}
   Distance Blend=@n4.Distance Blend
   Pre Skinned Local Position=ShaderStageSwitch(PixelShader=VertexInterpolator(VS=@n5).PS, VertexShader=@n5)
   Material=@n4.Distance Blend
  ) {materialFunction=&BaseFN_CharacterFunction_PostFX}
  SetMaterialAttributes!(MaterialAttributes=@n6.Result, Opacity Mask=@n2.G) {attributeSetTypes [#3]}
 ]
}
